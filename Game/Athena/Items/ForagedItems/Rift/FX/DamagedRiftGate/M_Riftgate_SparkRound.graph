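# Unlit translucent Niagara sprite material for round sparks, named for the damaged Rift Gate effect.
# A sphere mask on centered, exponent-shaped UVs forms the spark, with optional panning-noise distortion that is off by default.
# Opacity combines particle alpha, depth fade and a view-angle fade; color blends two tints per particle, and particle speed blends in a dimming vertical-falloff mask.
# Mobile drops depth fade, random tint and speed blending.
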Material /Game/Athena/Items/ForagedItems/Rift/FX/DamagedRiftGate/M_Riftgate_SparkRound
graph {
 properties {
  bEnableResponsiveAA=true
  bUsedWithNiagaraSprites=true
  blendMode=BLEND_Translucent
  pixelDepthOffsetMode=PDOM_Legacy
  shadingModel=MSM_Unlit
  usedShadingModels=Unlit
 }
 nodes [
  n1 = ParticleColor()
  n2 = Time()
  n3 = ParticleRandom()
  n4 = TextureSample(UVs=Panner(Coordinate=TextureCoordinate() {vTiling=0.5}, Time=Add(A=@n2, B=@n3)) {speedY=0.5}) {samplerType=SAMPLERTYPE_Grayscale, texture=&LowResBlurredNoise}
  n5 = ScalarParameter() {parameterName=UseDistortion}
  n6 = Add(A=TextureCoordinate(), B=Multiply(A=LinearInterpolate(Alpha=@n4.R) {constA=-1}, B=LinearInterpolate(Alpha=@n5) {constB=0.1}))
  n7 = MaterialFunctionCall(Float3=AppendVector(A=Subtract(A=@n6) {constB=0.5}, B=Constant())) {materialFunction=&BreakOutFloat3Components}
  n8 = Power(Base=Abs(None=@n7.R), Exp=ScalarParameter() {defaultValue=1.2, parameterName=U Exponent}) {constExponent=1}
  n9 = Abs(None=@n7.G)
  n10 = Power(Base=@n9, Exp=ScalarParameter() {defaultValue=2, parameterName=V Exponent})
  n11 = Multiply(
   A=@n1.A
   B=SphereMask(
    A=MaterialFunctionCall(X=@n8, Y=@n10) {materialFunction=&MakeFloat2}.Result
    B=Constant()
    Radius=ScalarParameter() {defaultValue=0.2, parameterName=Spark Radius}
    Hardness=ScalarParameter() {defaultValue=0.6, parameterName=Spark Hardness}
   )
  )
  n12 = ScalarParameter() {defaultValue=5, group=Opacity, parameterName=Depth Fade Distance}
  n13 = DotProduct(A=CameraVectorWS(), B=VertexNormalWS())
  Multiply(A=ShadingPathSwitch(Default=DepthFade(Opacity=@n11, FadeDistance=@n12), Mobile=@n11), B=Power(Base=Abs(None=@n13)) {constExponent=1.5})
  n14 = VectorParameter() {parameterName=Color A, defaultValue=<rgba 1 0.5534 0.295 0>}
  n15 = LinearInterpolate(A=@n14.RGB, B=VectorParameter() {parameterName=Color B, defaultValue=<rgba 0.401 0.3782 1 0>}.RGB, Alpha=ParticleRandom())
  n16 = Multiply(
   A=ScalarParameter() {defaultValue=0.5, parameterName=Speed Opacity}
   B=SphereMask(
    A=@n9
    B=Constant()
    Radius=ScalarParameter() {defaultValue=0.5, parameterName=Speed Mask Radius}
    Hardness=ScalarParameter() {parameterName=Speed Mask Hardness}
   ) {hardnessPercent=0}
  )
  n17 = ParticleSpeed()
  n18 = ScalarParameter() {defaultValue=300, parameterName=Speed LERP Divisor}
  n19 = LinearInterpolate(B=@n16, Alpha=Clamp(None=Divide(A=@n17, B=@n18))) {constA=1}
  Multiply(A=Multiply(A=ShadingPathSwitch(Default=@n15, Mobile=@n14.RGB), B=ShadingPathSwitch(Default=@n19, Mobile=@n16)), B=@n1.RGB)
 ]
}
